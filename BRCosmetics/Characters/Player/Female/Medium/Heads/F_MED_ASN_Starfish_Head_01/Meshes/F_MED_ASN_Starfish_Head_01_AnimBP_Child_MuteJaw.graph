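# Child animation Blueprint for the female Starfish head (MuteJaw) on the Fortnite base head skeleton, adjusting a head material for a specific cosmetic.
# On each animation update it gets the dynamic material instance at element 1 of the part's skeletal mesh.
# If the character item definition carries the BannerRedFemale meta tag, it writes the current LOD index into the "Hide Element 05" scalar parameter, tying that material element's hide value to LOD.
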AnimBlueprint /BRCosmetics/Characters/Player/Female/Medium/Heads/F_MED_ASN_Starfish_Head_01/Meshes/F_MED_ASN_Starfish_Head_01_AnimBP_Child_MuteJaw
target_skeleton=/Game/Characters/Player/Common/Fortnite_Base_Head/Fortnite_Base_Head_Export_Skeleton
graphs [EventGraph {nodes [
   Event BlueprintUpdateAnimation()
   > n1 = Get Dynamic Material Instance(SkeletalMeshComponent=GetPartSkeletalMeshComponent(), ElementIndex=1)
   > Branch(Condition=CharacterItemDefinitionHasAnyMetaTag(Tags=Make Array("[0]"=(TagName="Cosmetics.MetaTags.BannerRedFemale"))))
   > SetScalarParameterValue(self=@n1, ParameterName=Hide Element 05, Value="To Float (Integer)"(InInt=Get CurrentLOD()))
  ]}]
variables []
