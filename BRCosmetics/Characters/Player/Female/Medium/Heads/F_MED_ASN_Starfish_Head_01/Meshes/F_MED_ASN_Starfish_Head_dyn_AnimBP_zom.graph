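# Post-pose dynamics Animation Blueprint for the Starfish female head, apparently the zombie variant, adding secondary motion to an input pose.
# Chained Anim Dynamics nodes apply angular springs to five hair bones and a faint linear spring to dyn_hair_01.
# A partly blended RigidBody node follows, borrowing the male Hector head's zombie physics asset.
# The update event sets jawRandomRot from velocity-scaled Perlin noise, but the AnimGraph never reads it.
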AnimBlueprint /BRCosmetics/Characters/Player/Female/Medium/Heads/F_MED_ASN_Starfish_Head_01/Meshes/F_MED_ASN_Starfish_Head_dyn_AnimBP_zom
target_skeleton=/Game/Characters/Player/Common/Fortnite_Base_Head/Fortnite_Base_Head_Export_Skeleton
graphs [
 AnimGraph {nodes [
   AnimGraphNode_SubInput_1 = Input Pose()
   n1 = Anim Dynamics - Bone: dyn_hair_05(ComponentPose=Local To Component(LocalPose=@AnimGraphNode_SubInput_1), Alpha=1, ExternalForce="0,0,0") {
    angularDampingOverride=0.99
    angularSpringConstant=10000
    bAngularSpring=true
    bGravityOverrideInSimSpace=true
    bOverrideAngularDamping=true
    numSolverIterationsPostUpdate=4
    numSolverIterationsPreUpdate=16
    alphaBoolBlend {blendOption=Linear}
    boundBone {boneName=dyn_hair_05}
    physicsBodyDefinitions [{
      sphereCollisionRadius=0
      boundBone {boneName=dyn_hair_05}
      boxExtents=<5 30 30>
      constraintSetup {angularTargetAxis=AxisZ, angularLimitsMax=<2 20 0>, angularLimitsMin=<-2 -5 -20>, angularTarget=<0 0 1>}
      localJointOffset=<-1 0 0>
     }]
   }
   n2 = Anim Dynamics - Bone: dyn_hair_05_a(ComponentPose=@n1, Alpha=1, ExternalForce="0,0,0") {
    angularDampingOverride=0.99
    angularSpringConstant=10000
    bAngularSpring=true
    bGravityOverrideInSimSpace=true
    bOverrideAngularDamping=true
    numSolverIterationsPostUpdate=4
    numSolverIterationsPreUpdate=16
    alphaBoolBlend {blendOption=Linear}
    boundBone {boneName=dyn_hair_05_a}
    physicsBodyDefinitions [{
      sphereCollisionRadius=0
      boundBone {boneName=dyn_hair_05_a}
      boxExtents=<5 30 30>
      constraintSetup {angularTargetAxis=AxisZ, angularLimitsMax=<2 20 10>, angularLimitsMin=<-2 -6 -20>, angularTarget=<0 0 1>}
      localJointOffset=<-1 0 0>
     }]
   }
   n3 = Anim Dynamics - Bone: dyn_hair_05_b(ComponentPose=@n2, Alpha=1, ExternalForce="0,0,0") {
    angularDampingOverride=0.99
    angularSpringConstant=10000
    bAngularSpring=true
    bGravityOverrideInSimSpace=true
    bOverrideAngularDamping=true
    numSolverIterationsPostUpdate=4
    numSolverIterationsPreUpdate=16
    alphaBoolBlend {blendOption=Linear}
    boundBone {boneName=dyn_hair_05_b}
    physicsBodyDefinitions [{
      sphereCollisionRadius=0
      boundBone {boneName=dyn_hair_05_b}
      boxExtents=<5 30 30>
      constraintSetup {angularTargetAxis=AxisZ, angularLimitsMax=<2 20 20>, angularLimitsMin=<-2 -20 -20>, angularTarget=<0 0 1>}
      localJointOffset=<-1 0 0>
     }]
   }
   n4 = Anim Dynamics - Bone: dyn_hair_04(ComponentPose=@n3, Alpha=1, ExternalForce="0,0,0") {
    angularDampingOverride=0.99
    angularSpringConstant=10000
    bAngularSpring=true
    bGravityOverrideInSimSpace=true
    bOverrideAngularDamping=true
    numSolverIterationsPostUpdate=4
    numSolverIterationsPreUpdate=16
    alphaBoolBlend {blendOption=Linear}
    boundBone {boneName=dyn_hair_04}
    physicsBodyDefinitions [{
      sphereCollisionRadius=0
      boundBone {boneName=dyn_hair_04}
      boxExtents=<5 30 30>
      constraintSetup {angularTargetAxis=AxisZ, angularLimitsMax=<2 20 20>, angularLimitsMin=<-2 -20 -20>, angularTarget=<0 0 1>}
      localJointOffset=<-1 0 0>
     }]
   }
   n5 = Anim Dynamics - Bone: dyn_hair_04_a(ComponentPose=@n4, Alpha=1, ExternalForce="0,0,0") {
    angularDampingOverride=0.99
    angularSpringConstant=10000
    bAngularSpring=true
    bGravityOverrideInSimSpace=true
    bOverrideAngularDamping=true
    numSolverIterationsPostUpdate=4
    numSolverIterationsPreUpdate=16
    alphaBoolBlend {blendOption=Linear}
    boundBone {boneName=dyn_hair_04_a}
    physicsBodyDefinitions [{
      sphereCollisionRadius=0
      boundBone {boneName=dyn_hair_04_a}
      boxExtents=<5 30 30>
      constraintSetup {angularTargetAxis=AxisZ, angularLimitsMax=<2 20 20>, angularLimitsMin=<-2 -20 -20>, angularTarget=<0 0 1>}
      localJointOffset=<-1 0 0>
     }]
   }
   n6 = Anim Dynamics - Bone: dyn_hair_01(ComponentPose=@n5, Alpha=0.05) {
    bGravityOverrideInSimSpace=true
    bLinearSpring=true
    bOverrideLinearDamping=true
    gravityScale=0.2
    lODThreshold=1
    linearDampingOverride=0.997
    linearSpringConstant=200
    simulationSpace=Actor
    alphaBoolBlend {blendOption=Linear}
    boundBone {boneName=dyn_hair_01}
    physicsBodyDefinitions [{
      sphereCollisionRadius=0
      boundBone {boneName=dyn_hair_01}
      boxExtents=<60 60 60>
      constraintSetup {
       twistAxis=AxisZ
       angularLimitsMax=<0 0 0>
       angularLimitsMin=<0 0 0>
       angularTarget=<0 0 0>
       linearAxesMax=<10 10 10>
       linearAxesMin=<-10 -10 -10>
      }
      localJointOffset=<0 0 0>
     }]
   }
   n7 = RigidBody(ComponentPose=@n6, Alpha=0.45) {
    bClampLinearTranslationLimitToRefPose=true
    bOverrideWorldGravity=true
    overridePhysicsAsset=/Game/Characters/Player/Male/Medium/Heads/M_MED_HIS_Hector_Head_01/Meshes/M_MED_Hector_01_Physics_zom
    alphaBoolBlend {blendOption=Linear}
    baseBoneRef {boneName=spine_05}
    componentLinearAccScale=<0.3 0.3 0.3>
    componentLinearVelScale=<0.3 0.3 0.3>
    overrideWorldGravity {z=-980}
   }
   Output Pose(Result=Component To Local(ComponentPose=@n7)) {layerGroup=None}
  ]}
 EventGraph {nodes [
   Event BlueprintUpdateAnimation()
   n1 = "Absolute (Float)"(A=BreakVector(InVec=Get PawnVelocity()).Y)
   > n2 = jawRotRandomFNC!(
    execute
    self
    frequency:double=float + float(A=float * float(A=@n1, B=0.01), B=2) {is=title K2Node_CommutativeAssociativeBinaryOperator}
    amp:double=20
   ) -> (then, output:double)
   > Set jawRandomRot(jawRandomRot:Rotator=MakeRotator(Roll=@n2)) -> (Output_Get:Rotator)
   TryGetPawnOwner()
  ]}
 jawRotRandomFNC {nodes [
   n1 = jawRotRandomFNC!() -> (then, frequency:double=~, amp:double=~) {is=title K2Node_FunctionEntry}
   n2 = PerlinNoise1D!(Value:float=float * float(A=@n1, B=GetGameTimeInSeconds())) -> (ReturnValue:float)
   n3 = float * float(A="Absolute (Float)"(A="Sin (Radians)"(A=@n2)), B=@n1.amp)
   > Return Node!(execute, output:double=@n3)
   PerlinNoise1D(Value=@n3)
  ]}
]
variables [jawRandomRot {type=Rotator}]
